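# Physics asset for the face accessory mesh of the CarolinaChili medium male player cosmetic body, setting up a three-bone branch chain on the head.
# Each dyn_branch bone carries a small capsule, and the head carries a larger capsule.
# Constraints link the branch bones in series back to the head, each with the same tight swing limits and a narrower twist limit, restricting how far the chain bends.
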PhysicsAsset /BRCosmetics/Characters/Player/Male/Medium/Bodies/M_MED_CarolinaChili/Meshes/Parts/M_MED_CarolinaChili_FaceAcc_Physics
bodies {
 dyn_branch_1 [{shapeType=Capsule, length=5, radius=2, shapeName=dyn_branch_1_capsule, center=<3.3287 0 0>, rotation=<rot 90 0 0>}]
 dyn_branch_2 [{shapeType=Capsule, length=5, radius=2, shapeName=dyn_branch_2_capsule, center=<3.3287 0 0>, rotation=<rot 90 0 0>}]
 dyn_branch_3 [{shapeType=Capsule, length=4.4626, radius=2.5374, shapeName=dyn_branch_3_capsule, center=<3.3287 0 0>, rotation=<rot 90 0 0>}]
 head [{shapeType=Capsule, length=6.2591, radius=11.7278, center=<6.0693 0 0>, rotation=<rot 90 0 0>}]
}
constraints [
 {bone1Name=dyn_branch_1, bone2Name=head, swing1LimitDegrees=15, swing2LimitDegrees=15, twistLimitDegrees=5}
 {bone1Name=dyn_branch_2, bone2Name=dyn_branch_1, swing1LimitDegrees=15, swing2LimitDegrees=15, twistLimitDegrees=5}
 {bone1Name=dyn_branch_3, bone2Name=dyn_branch_2, swing1LimitDegrees=15, swing2LimitDegrees=15, twistLimitDegrees=5}
]
